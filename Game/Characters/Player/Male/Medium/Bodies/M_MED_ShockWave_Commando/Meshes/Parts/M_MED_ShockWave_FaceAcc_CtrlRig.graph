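# Control Rig for the ShockWave Commando face accessory mesh that adds a corrective head rotation from the neck pose.
# It reads the head rotation relative to neck_01 as Euler angles and clamp-remaps Z past -25 degrees into a positive offset, up to 25 at -50.
# That offset is applied as an additive local rotation on head, countering extreme bend.
# The hierarchy includes dyn_ribbon bones, but the graph does not drive them directly.
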ControlRigBlueprint /Game/Characters/Player/Male/Medium/Bodies/M_MED_ShockWave_Commando/Meshes/Parts/M_MED_ShockWave_FaceAcc_CtrlRig
graphs [
 RigVMFunctionLibrary {links [], nodes []}
 RigVMModel {nodes [
   RigUnit_BeginExecution = &RigUnit_BeginExecution()
   n1 = &RigVMFunction_MathQuaternionToEuler(RotationOrder=ZYX, Value=&RigUnit_GetRelativeBoneTransform(Bone=head, Space=neck_01).Transform.Rotation)
   /Script/ControlRig.RigUnit_ModifyBoneTransforms(
    BoneToModify=((Bone="head",Transform=(Rotation=(X=0,Y=0,Z=0,W=1),Translation=(X=0,Y=0,Z=0),Scale3D=(X=1,Y=1,Z=1))))
    Mode=AdditiveLocal
    Weight=1
    WeightMaximum=1
    WeightMinimum=0
    BoneToModify.0.Transform.Rotation=&RigVMFunction_MathQuaternionFromEuler(
     Euler=(X=0,Y=0,Z=0)
     Euler.Z=&RigVMFunction_MathFloatRemap(SourceMaximum=-50, SourceMinimum=-25, TargetMaximum=25, TargetMinimum=0, bClamp=true, Value=@n1.Result.Z).Result
    ).Result
    ExecutePin=@RigUnit_BeginExecution.ExecutePin
   ) {is=struct RigVMUnitNode}
  ]}
]
hierarchy [
 root {loc=<0 0 0>, rot=<0 0 0 1>}
 pelvis {parent=root, loc=<0 -0.7071 93.1191>, rot=<0.0127 0.707 -0.0127 -0.707>}
 spine_01 {parent=pelvis, loc=<6.7494 -0.6388 0>, rot=<0 0 0.034 -0.9994>}
 spine_02 {parent=spine_01, loc=<7.2518 -0.035 0>, rot=<0 0 0.0238 -0.9997>}
 spine_03 {parent=spine_02, loc=<6.0104 0 0>, rot=<0 0 0.0149 -0.9999>}
 spine_04 {parent=spine_03, loc=<8.1024 0 0>, rot=<0 0 -0.0132 -0.9999>}
 spine_05 {parent=spine_04, loc=<10.0146 0 0>, rot=<0 0 -0.0015 -1>}
 clavicle_l {parent=spine_05, loc=<4.3575 -0.1466 -2.3223>, rot=<0.0465 -0.7289 -0.009 -0.6829>}
 clavicle_r {parent=spine_05, loc=<4.3575 -0.1466 2.3223>, rot=<0.7289 0.0465 -0.6829 0.009>}
 neck_01 {parent=spine_05, loc=<11.3138 0.1529 0>, rot=<0 0 -0.2003 -0.9797>}
 neck_02 {parent=neck_01, loc=<4.5627 0 0>, rot=<0 0 0.0504 -0.9987>}
 head {parent=neck_02, loc=<5.3914 0 0>, rot=<0 0 0.115 -0.9934>}
 dyn_head {parent=head, loc=<0 0 0>, rot=<0 0 0 -1>}
 dyn_ribbon_mstr {parent=dyn_head, loc=<4.5949 0.7002 0>, rot=<0.0029 -0.7071 -0.0029 -0.7071>}
 dyn_ribbon_bk_01 {parent=dyn_ribbon_mstr, loc=<0 -2.6601 8.0269>, rot=<0 0 0 1>}
 dyn_ribbon_bk_02 {parent=dyn_ribbon_bk_01, loc=<3.6197 -6.9429 -0.8921>, rot=<0 0 0 1>}
 dyn_ribbon_l_01 {parent=dyn_ribbon_mstr, loc=<-0.3656 0.5682 7.4288>, rot=<0 0 0 1>}
 dyn_ribbon_l_02 {parent=dyn_ribbon_l_01, loc=<6.5933 -0.5262 1.4695>, rot=<0 0 0 1>}
]
