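# Physics asset for the face accessory mesh of the male medium Shock Wave Commando player character, covering its dynamic ribbon bones.
# A master ribbon body combines a small capsule with a larger box, while the back and left ribbon bones each have one capsule.
# Both ribbon bodies are constrained directly to the master with limited swing and nearly locked twist, and the left ribbon is tighter on one swing axis.
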PhysicsAsset /Game/Characters/Player/Male/Medium/Bodies/M_MED_ShockWave_Commando/Meshes/Parts/M_MED_Shock_Wave_Commando_FaceAcc_Physics
bodies {
 dyn_ribbon_bk_01 [{shapeType=Capsule, length=5.5021, radius=1.6121, center=<0.9432 -2.5438 0.6129>, rotation=<rot -0.0053 -152.61 89.9969>}]
 dyn_ribbon_l_01 [{shapeType=Capsule, length=4.0698, radius=1.5143, center=<1.4872 -0.0001 1.2557>, rotation=<rot 0 -89.9999 90.7149>}]
 dyn_ribbon_mstr [
  {shapeType=Capsule, length=0.505, radius=0.505, rotation=<rot 0 -90 0>}
  {shapeType=Box, extentX=9, extentY=14, extentZ=5.5205, center=<0 0 2.0119>}
 ]
}
constraints [
 {bone1Name=dyn_ribbon_bk_01, bone2Name=dyn_ribbon_mstr, swing1LimitDegrees=25, swing2LimitDegrees=25, twistLimitDegrees=1}
 {bone1Name=dyn_ribbon_l_01, bone2Name=dyn_ribbon_mstr, swing1LimitDegrees=5, swing2LimitDegrees=25, twistLimitDegrees=1}
]
