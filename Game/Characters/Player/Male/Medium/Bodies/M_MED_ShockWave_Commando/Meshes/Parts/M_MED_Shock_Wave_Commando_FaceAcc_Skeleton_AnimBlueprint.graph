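# Animation Blueprint for the face accessory part of the male medium ShockWave Commando player body, adding secondary motion on top of the body's animation.
# The pose copies from the attached parent mesh, runs through a Control Rig, then a RigidBody simulation in spine_05 bone space.
# RigidBody alpha drops from full to 0.7 while bIsTargeting is true.
# A chain of three ribbon AnimDynamics nodes is left unconnected to the output, and the EventGraph has no logic.
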AnimBlueprint /Game/Characters/Player/Male/Medium/Bodies/M_MED_ShockWave_Commando/Meshes/Parts/M_MED_Shock_Wave_Commando_FaceAcc_Skeleton_AnimBlueprint
target_skeleton=&/M_MED_Shock_Wave_Commando_FaceAcc_Skeleton
graphs [
 AnimGraph {nodes [
   n1 = Anim Dynamics - Bone: dyn_ribbon_mstr(Alpha=0.1, ExternalForce="0,0,0") {
    angularDampingOverride=0.98
    angularSpringConstant=90
    bAngularSpring=true
    bGravityOverrideInSimSpace=true
    bLinearSpring=true
    bOverrideLinearDamping=true
    gravityScale=0.05
    lODThreshold=1
    linearDampingOverride=0.98
    linearSpringConstant=100
    simulationSpace=Actor
    alphaBoolBlend {blendOption=Linear}
    alphaScaleBiasClamp {
     bClampResult=true
     bInterpResult=true
     clampMax=0.25
     clampMin=0.05
     interpSpeedDecreasing=0.5
     interpSpeedIncreasing=0.5
     outRange {max=0.1}
    }
    boundBone {boneName=dyn_ribbon_mstr}
    physicsBodyDefinitions [{
      sphereCollisionRadius=0
      boundBone {boneName=dyn_ribbon_mstr}
      boxExtents=<30 30 30>
      constraintSetup {
       angularTargetAxis=AxisY
       twistAxis=AxisZ
       angularLimitsMax=<30 30 0>
       angularLimitsMin=<-30 -30 0>
       angularTarget=<0 0 0>
       linearAxesMax=<1 1 5>
       linearAxesMin=<-1 -1 -1>
      }
      localJointOffset=<0 0 0>
     }]
   }
   n2 = Anim Dynamics - Bone: dyn_ribbon_l_01(ComponentPose=@n1, Alpha=0.1, ExternalForce="0,0,0") {
    angularDampingOverride=0.98
    angularSpringConstant=90
    bAngularSpring=true
    bGravityOverrideInSimSpace=true
    bLinearSpring=true
    bOverrideLinearDamping=true
    gravityScale=0.05
    lODThreshold=1
    linearDampingOverride=0.98
    linearSpringConstant=100
    simulationSpace=Actor
    alphaBoolBlend {blendOption=Linear}
    alphaScaleBiasClamp {
     bClampResult=true
     bInterpResult=true
     clampMax=0.25
     clampMin=0.05
     interpSpeedDecreasing=0.5
     interpSpeedIncreasing=0.5
     outRange {max=0.1}
    }
    boundBone {boneName=dyn_ribbon_l_01}
    physicsBodyDefinitions [{
      sphereCollisionRadius=0
      boundBone {boneName=dyn_ribbon_l_01}
      boxExtents=<30 30 30>
      constraintSetup {
       angularTargetAxis=AxisY
       twistAxis=AxisZ
       angularLimitsMax=<2 20 0>
       angularLimitsMin=<-2 -20 0>
       angularTarget=<0 0 0>
       linearAxesMax=<1 1 5>
       linearAxesMin=<-1 -1 -1>
      }
      localJointOffset=<0 0 0>
     }]
   }
   Anim Dynamics - Bone: dyn_ribbon_bk_01(ComponentPose=@n2, Alpha=0.1, ExternalForce="0,0,0") {
    angularDampingOverride=0.98
    angularSpringConstant=90
    bAngularSpring=true
    bGravityOverrideInSimSpace=true
    bLinearSpring=true
    bOverrideLinearDamping=true
    gravityScale=0.05
    lODThreshold=1
    linearDampingOverride=0.98
    linearSpringConstant=100
    simulationSpace=Actor
    alphaBoolBlend {blendOption=Linear}
    alphaScaleBiasClamp {
     bClampResult=true
     bInterpResult=true
     clampMax=0.25
     clampMin=0.05
     interpSpeedDecreasing=0.5
     interpSpeedIncreasing=0.5
     outRange {max=0.1}
    }
    boundBone {boneName=dyn_ribbon_bk_01}
    physicsBodyDefinitions [{
      sphereCollisionRadius=0
      boundBone {boneName=dyn_ribbon_bk_01}
      boxExtents=<30 30 30>
      constraintSetup {
       angularTargetAxis=AxisY
       twistAxis=AxisZ
       angularLimitsMax=<30 2 0>
       angularLimitsMin=<-30 -2 0>
       angularTarget=<0 0 0>
       linearAxesMax=<1 1 5>
       linearAxesMin=<-1 -1 -1>
      }
      localJointOffset=<0 0 0>
     }]
   }
   n3 = Control Rig(Source=Copy Pose From Mesh() {bUseAttachedParent=true}) {
    alphaBoolBlend {blendOption=Linear}
    controlRigAssetReference {blueprintRigClass=&/M_MED_ShockWave_FaceAcc_CtrlRig.M_MED_ShockWave_FaceAcc_CtrlRig_C}
   }
   n4 = RigidBody(ComponentPose=Local To Component(LocalPose=@n3), Alpha=SelectFloat(A=0.7, B=1, bPickA=Get bIsTargeting())) {
    bClampLinearTranslationLimitToRefPose=true
    bOverrideWorldGravity=true
    overridePhysicsAsset=&/M_MED_Shock_Wave_Commando_FaceAcc_Physics
    simulationSpace=BaseBoneSpace
    alphaBoolBlend {blendOption=Linear}
    baseBoneRef {boneName=spine_05}
    componentAppliedLinearAccClamp {z=2000}
    componentLinearAccScale=<0.35 0.35 0.65>
    componentLinearVelScale=<0.35 0.35 0.55>
    externalForce=<-30000 1000 1000>
    overrideWorldGravity {z=-980}
   }
   Output Pose(Result=Component To Local(ComponentPose=@n4)) {layerGroup=None}
  ]}
 EventGraph {nodes [
   TryGetPawnOwner()
   Event BlueprintUpdateAnimation()
  ]}
]
variables []
